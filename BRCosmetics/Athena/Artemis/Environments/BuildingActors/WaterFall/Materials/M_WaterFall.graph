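# M_WaterFall is a masked material for waterfall building actors, with animated foam, color and opacity.
# Curve atlases and scrolling packed noise and foam textures shape UVs, foam, normals and color, with time offset by object position.
# Custom HLSL evaluates two-band spherical harmonics from vertex color for tinting and emissive.
# Normal-map noise drives distance-faded vertex offset, disabled on mobile, while temporal dither softens ends and near-camera pixels.
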Material /BRCosmetics/Athena/Artemis/Environments/BuildingActors/WaterFall/Materials/M_WaterFall
graph {
 properties {
  bUseMaterialAttributes=true
  blendMode=BLEND_Masked
  pixelDepthOffsetMode=PDOM_Legacy
  translucencyLightingMode=TLM_VolumetricPerVertexDirectional
  usedShadingModels=Default Lit
 }
 nodes [
  n1 = ScalarParameter() {group=00.Color, parameterName=Color_HueShift}
  n2 = NamedRerouteUsage()
  n3 = NamedRerouteUsage()
  n4 = Add(
   A=ScalarParameter() {group=000.USE ABOVE PARAMS TO TWEAK COLOR, parameterName=_}
   B=ScalarParameter() {group=001.FOLLOWING PARAMS ARE NOT INTUITIVE, parameterName=__}
  )
  n5 = ScalarParameter() {group=06.Opacity, parameterName=OpacityPropertyDebug}
  n6 = NamedRerouteUsage()
  n7 = NamedRerouteUsage()
  MakeMaterialAttributes!(
   BaseColor=LinearInterpolate(
    A=MaterialFunctionCall(Hue Shift Percentage=@n1, Texture=@n2) {materialFunction=&HueShift}.Result
    B=@n3
    Alpha=Add(A=@n4, B=@n5)
   )
   Metallic=ScalarParameter() {group=07.Other, parameterName=Metallic}
   Specular=ScalarParameter() {group=07.Other, parameterName=Specular}
   Roughness=NamedRerouteUsage()
   Anisotropy
   EmissiveColor=MaterialFunctionCall(Hue Shift Percentage=@n1, Texture=NamedRerouteUsage()) {materialFunction=&HueShift}.Result
   Opacity=@n3
   OpacityMask=@n6
   Normal=NamedRerouteUsage()
   Tangent
   WorldPositionOffset=ShadingPathSwitch(Default=@n7, Deferred=@n7, Forward=@n7, Mobile=Constant())
   SubsurfaceColor
   ClearCoat
   ClearCoatRoughness
   AmbientOcclusion
   Refraction
   CustomizedUVs_0
   CustomizedUVs_1
   CustomizedUVs_2
   CustomizedUVs_3
   CustomizedUVs_4
   CustomizedUVs_5
   CustomizedUVs_6
   CustomizedUVs_7
   PixelDepthOffset
   ShadingModel
   Displacement
  )
  n8 = VectorParameter() {group=00.UV, parameterName=UV_ScaleUVByCenter, defaultValue=<rgba 1 1 0 0>}
  n9 = MaterialFunctionCall(UVs=TextureCoordinate(), Texture Scale=AppendVector(A=@n8.R, B=@n8.R)) {materialFunction=&ScaleUVsByCenter}
  NamedRerouteDeclaration(None=MaterialFunctionCall(
    X=MaterialFunctionCall(Float2=@n9.UVs) {materialFunction=&BreakOutFloat2Components}.R
    Y=NamedRerouteUsage()
   ) {materialFunction=&MakeFloat2}.Result) {name=RemapedUV, nodeColor=<rgba 1 0 0.9609 1>}
  n10 = Fmod(
   A=MaterialFunctionCall(Float3=ObjectPositionWS()) {materialFunction=&BreakOutFloat3Components}.R
   B=ScalarParameter() {defaultValue=13.256, group=00.Time, parameterName=Time_RandomMod}
  )
  n11 = Time() {bOverride_Period=true, period=330}
  n12 = ScalarParameter() {defaultValue=0.7, group=00.Time, parameterName=Time_ScaleTotalSpeed}
  NamedRerouteDeclaration(None=Add(A=Multiply(A=@n10) {constB=15.25}, B=Multiply(A=@n11, B=@n12))) {name=Time, nodeColor=<rgba 0.9558 1 0 1>}
  n13 = VertexColor()
  n14 = AppendVector(A=@n13, B=@n13)
  n15 = VectorParameter() {group=03.SHLighting, parameterName=SH_FitMax, defaultValue=<rgba 1.25 0.13 0.18 0.13>}
  n16 = VectorParameter() {group=03.SHLighting, parameterName=SH_FitMin, defaultValue=<rgba 1 -0.16 -0.17 -0.15>}
  n17 = AppendVector(A=@n16.RGB, B=@n16.RGB)
  n18 = Subtract(A=AppendVector(A=@n15.RGB, B=@n15.RGB), B=@n17)
  n19 = Multiply(A=MaterialFunctionCall(X=@n13.G, Y=@n13.G, Z=@n13.G) {materialFunction=&MakeFloat3}.Result)
  MaterialExpressionStaticSwitchParameter_1 = StaticSwitchParameter(True=Add(A=Multiply(A=@n14, B=@n18), B=@n17), False=AppendVector(A=Add(A=@n13.R), B=Add(A=@n19) {constB=-0.5})) {group=03.SHLighting, parameterName=UseCustomSH Fit}
  n20 = Subtract(A=CameraPositionWS(), B=WorldPosition().XYZ)
  n21 = MaterialFunctionCall(Float3=NamedRerouteUsage()) {materialFunction=&BreakOutFloat3Components}
  n22 = Transform(None=MaterialFunctionCall(X=@n21.R, Y=@n21.R, Z=@n21.R) {materialFunction=&MakeFloat3}.Result)
  n23 = ScalarParameter() {group=03.SHLighting, parameterName=SH_NormalStrength}
  MaterialExpressionTransform_0 = Transform(None=Subtract(A=Normalize(VectorInput=@n20), B=Multiply(A=@n22, B=@n23))) {transformSourceType=TRANSFORMSOURCE_World, transformType=TRANSFORM_Local}
  n24 = Custom!(SH=@MaterialExpressionStaticSwitchParameter_1, Direction=@MaterialExpressionTransform_0) {
   code="//SH decode,  Direction = LightDir, SH = 2band coefficients. (encoded in Y-up)\r\nDirection = normalize(Direction);\r\nfloat4 shEvalDirection;\r\nshEvalDirection.x = 0.282095;\r\nshEvalDirection.y =0.488603 * Direction.z;\r\nshEvalDirection.z = 0.488603 * Direction.y;\r\nshEvalDirection.w =-0.488603 * Direction.x;\r\nreturn saturate(dot(SH, shEvalDirection));"
   description=SphericalHarmonicsEval
   outputType=CMOT_Float1
   inputs [
    {
     inputName=SH
     input {expression=&self:MaterialExpressionStaticSwitchParameter_1}
    }
    {
     inputName=Direction
     input {expression=&self:MaterialExpressionTransform_0}
    }
   ]
  }
  n25 = MaterialFunctionCall(
   Input=@n24
   Input Low=ScalarParameter() {group=03.SHLighting, parameterName=SH_RemapLowValue}
   Input High=ScalarParameter() {defaultValue=1, group=03.SHLighting, parameterName=SH_RemapHighValue}
   Target Low=Constant()
   Target High=Constant() {r=1}
  ) {materialFunction=&RemapValueRange}
  n26 = Power(Base=Saturate(None=OneMinus(None=@n25.Result)), Exp=ScalarParameter() {defaultValue=1, group=03.SHLighting, parameterName=SH_Pow})
  n27 = ScalarParameter() {group=03.SHLighting, parameterName=SH_Contrast}
  NamedRerouteDeclaration(None=LinearInterpolate(
    B=MaterialFunctionCall(In=@n26, Contrast=@n27) {materialFunction=&CheapContrast}.Result
    Alpha=NamedRerouteUsage()
   )) {name=SH_Value, nodeColor=<rgba 0.5703 1 0 1>}
  n28 = CurveAtlasRowParameter(CurveTime=MaterialFunctionCall(Float2=NamedRerouteUsage()) {materialFunction=&BreakOutFloat2Components}.G) {
   atlas=/Game/Athena/Artemis/Environments/BuildingActors/WaterFall/Materials/Texture/CurveAtlas/CLA_WaterFall
   curve=/Game/Athena/Artemis/Environments/BuildingActors/WaterFall/Materials/Texture/CurveAtlas/CL_UVCurve
   group=80.Curves
   parameterName=Curve_Remap/Opacity/SHMask
  }
  NamedRerouteDeclaration(None=@n28.B) {name=CurveA_SHMask, nodeColor=<rgba 0.4531 0 1 1>}
  NamedRerouteDeclaration(None=@n28.G) {name=CurveA_Opacity, nodeColor=<rgba 0.4531 0 1 1>}
  n29 = CameraPositionWS()
  n30 = ObjectPositionWS()
  n31 = Subtract(
   A=MaterialFunctionCall(Vector 3=Subtract(A=@n29, B=@n30)) {materialFunction=/Engine/Functions/Engine_MaterialFunctions02/Utility/VectorLength}.V3 Length
   B=ScalarParameter() {defaultValue=2048, parameterName=MinWPOLength}
  )
  n32 = ScalarParameter() {defaultValue=5120, parameterName=MaxWPOLength}
  n33 = Power(Base=Saturate(None=Divide(A=@n31, B=@n32)))
  n34 = NamedRerouteUsage()
  n35 = VectorParameter() {group=20.WPO, parameterName=WPO_ScaleAndSpeed, defaultValue=<rgba 1 1 0 0>}
  n36 = AppendVector(A=@n35.R, B=@n35.R)
  n37 = AppendVector(A=@n35.B, B=@n35.B)
  n38 = NamedRerouteUsage()
  n39 = TextureSampleParameter2D(UVs=Add(A=Multiply(A=@n34, B=@n36), B=Multiply(A=@n37, B=@n38))) {
   parameterName=WPO_Texture
   samplerType=SAMPLERTYPE_Normal
   texture=/Game/Characters/Player/Male/Large/Bodies/M_LRG_Clash_V/Textures/Simple_Perlin_N
  }
  n40 = Multiply(
   A=VectorParameter() {group=20.WPO, parameterName=WPO_TangentSpaceScale, defaultValue=<rgba 0 0 0 1>}.RGB
   B=Add(A=@n39.RGB, B=Constant3Vector() {constant=<rgba 0 0 -1 1>}) {constB=-0.5}
  )
  n41 = Multiply(A=@n40, B=Constant3Vector() {constant=<rgba -1 -1 1 0>})
  n42 = If(
   A=MaterialFunctionCall(Float2=TextureCoordinate() {coordinateIndex=1}) {materialFunction=&BreakOutFloat2Components}.R
   B=Constant() {r=0.2}
   "A > B"=@n40
   "A == B"=@n41
   "A < B"=@n41
  )
  n43 = ScalarParameter() {parameterName=WPO_EnableTangentFix}
  n44 = MaterialFunctionCall(Float2=TextureCoordinate()) {materialFunction=&BreakOutFloat2Components}
  n45 = VectorParameter() {group=20.WPO, parameterName=WPO_Mask, channelNames {a="", b="", g=Multiply, r=Offset}, defaultValue=<rgba -0.17 3 0 1>}
  n46 = Multiply(A=Add(A=OneMinus(None=@n44.G), B=@n45.Offset) {constB=-0.17}, B=@n45.Multiply) {constB=3}
  n47 = Multiply(
   A=Transform(None=LinearInterpolate(A=@n40, B=@n42, Alpha=@n43)) {transformType=TRANSFORM_Local}
   B=LinearInterpolate(
    A=VectorParameter() {group=20.WPO, parameterName=WPO_TopStrengthScale, defaultValue=<rgba 0 0 0 1>}.RGB
    B=VectorParameter() {group=20.WPO, parameterName=WPO_BottomStrengthScale, defaultValue=<rgba 0 0 0 1>}.RGB
    Alpha=Saturate(None=@n46)
   )
  )
  NamedRerouteDeclaration(None=Multiply(A=OneMinus(None=@n33), B=Transform(None=@n47) {transformSourceType=TRANSFORMSOURCE_Local})) {name=WPO_Value, nodeColor=<rgba 0 0.9297 1 1>}
  n48 = MaterialFunctionCall(Float2=TextureCoordinate()) {materialFunction=&BreakOutFloat2Components}
  n49 = VectorParameter() {group=04.OpacityMask, parameterName=OpacityMask_SurfaceBlendArea, channelNames {a="", b="", g=Offset, r=Width}}
  n50 = Multiply(A=Add(A=@n48.G, B=ShadowReplace(Default=@n49.Offset, Shadow=@n49.Offset)), B=ShadowReplace(Default=@n49.Width, Shadow=@n49.Width))
  n51 = VectorParameter() {group=04.OpacityMask, parameterName=OpacityMask_BottomDither, channelNames {a="", b="", g=Offset, r=Width}, defaultValue=<rgba 22 -0.01 0 0>}
  n52 = Add(A=Saturate(None=@n48.G), B=@n51.Offset)
  n53 = Multiply(A=Saturate(None=OneMinus(None=@n50)), B=Saturate(None=Multiply(A=@n52, B=@n51.Width)))
  n54 = VectorParameter() {
   group=04.OpacityMask
   parameterName=OpacityMask_TopBackfaceDither
   channelNames {a="", b="", g=Offset, r=Width}
   defaultValue=<rgba 22 -0.01 0 0>
  }
  n55 = Multiply(A=Add(A=@n48.G, B=@n54.Offset), B=@n54.Width)
  n56 = If(
   A=MaterialFunctionCall(Float2=TextureCoordinate() {coordinateIndex=1}) {materialFunction=&BreakOutFloat2Components}.R
   B=Constant() {r=0.2}
   "A > B"=Constant()
   "A == B"=@n55
   "A < B"=@n55
  )
  n57 = Multiply(A=@n53, B=OneMinus(None=Saturate(None=@n56)))
  n58 = ScalarParameter() {parameterName=OpacityMask_EnableTopBackfaceDither}
  n59 = Saturate(None=MaterialFunctionCall(Alpha Threshold=LinearInterpolate(A=@n53, B=@n57, Alpha=@n58), Random=Constant()) {materialFunction=&DitherTemporalAA_Fortnite}.Result)
  n60 = Add(A=Subtract(A=NamedRerouteUsage()), B=NamedRerouteUsage())
  NamedRerouteDeclaration(None=Saturate(None=Multiply(A=@n59, B=@n60))) {name=Final_OpacityMask, nodeColor=<rgba 1 0 0.7266 1>}
  n61 = MaterialFunctionCall(Float2=TextureCoordinate()) {materialFunction=&BreakOutFloat2Components}
  n62 = Multiply(A=@n61.R) {constB=0.5}
  n63 = Saturate(None=Multiply(A=Sine(None=@n62), B=ScalarParameter() {defaultValue=10, group=06.Opacity, parameterName=Opacity_EdgeFade}))
  n64 = ScalarParameter() {defaultValue=2, group=06.Opacity, parameterName=Opacity_EdgeFadePow}
  n65 = MaterialFunctionCall(Float2=TextureCoordinate()) {materialFunction=&BreakOutFloat2Components}
  n66 = VectorParameter() {group=06.Opacity, parameterName=OpacityC_TailFade, channelNames {a="", b="", g=Width, r=Offset}}
  n67 = Saturate(None=Multiply(A=Add(A=@n65.G, B=@n66.Width), B=@n66.Offset))
  n68 = Saturate(None=NamedRerouteUsage())
  n69 = VectorParameter() {
   group=06.Opacity
   parameterName=OpacityA_Attribute
   channelNames {a=SubBeforeMul, b=Contrast, g=SubAfterMul, r=Multiply}
   defaultValue=<rgba 0 0 0 0.2>
  }
  n70 = Subtract(A=@n68, B=@n69.SubBeforeMul)
  n71 = Subtract(A=Multiply(A=Saturate(None=@n70), B=@n69.Multiply), B=@n69.SubAfterMul)
  n72 = VectorParameter() {
   group=06.Opacity
   parameterName=OpacityB_Attribute
   channelNames {a=SubBeforeMul, b=Contrast, g=SubAfterMul, r=Multiply}
   defaultValue=<rgba 0 0 0 0.2>
  }
  n73 = Subtract(A=@n68, B=@n72.SubBeforeMul)
  n74 = Subtract(A=Multiply(A=Saturate(None=@n73), B=@n72.Multiply), B=@n72.SubAfterMul)
  n75 = VectorParameter() {group=06.Opacity, parameterName=Opacity_LerpAlphaTweak, channelNames {a="", b="", g=Offset, r=Width}}
  n76 = Add(A=@n61.G, B=@n75.Offset)
  n77 = LinearInterpolate(
   A=MaterialFunctionCall(In=Saturate(None=@n71), Contrast=@n69.Contrast) {materialFunction=&CheapContrast}.Result
   B=MaterialFunctionCall(In=Saturate(None=@n74), Contrast=@n72.Contrast) {materialFunction=&CheapContrast}.Result
   Alpha=Saturate(None=Multiply(A=@n76, B=@n75.Width))
  )
  NamedRerouteDeclaration(None=Multiply(A=Power(Base=@n63, Exp=@n64), B=Multiply(A=@n67, B=@n77))) {name=Final_Opacity, nodeColor=<rgba 1 0.4453 0 1>}
  n78 = Power(Base=NamedRerouteUsage(), Exp=ScalarParameter() {group=01.FoamMask, parameterName=FoamMask_Power})
  n79 = Subtract(A=Multiply(A=@n78) {constB=2})
  n80 = MaterialFunctionCall(Float2=TextureCoordinate()) {materialFunction=&BreakOutFloat2Components}
  n81 = OneMinus(None=@n80.R)
  n82 = Multiply(A=Subtract(A=Max(A=@n80.R, B=@n81)) {constB=0.5}) {constB=2}
  n83 = VectorParameter() {group=01.FoamMask, parameterName=FoamMask_EdgeMaskAttributes, defaultValue=<rgba 0 0 0 1>}
  n84 = MaterialFunctionCall(In=Power(Base=Saturate(None=@n82), Exp=@n83.R) {constExponent=5}, Contrast=@n83.G) {materialFunction=&CheapContrast}
  n85 = Saturate(None=Add(A=NamedRerouteUsage(), B=Add(A=@n79, B=@n84.Result)))
  NamedRerouteDeclaration(None=@n85) {name=Roughness, nodeColor=<rgba 1 0.9375 0 1>}
  n86 = NamedRerouteUsage()
  n87 = VectorParameter() {group=00.Color, parameterName=Color_SHColor, defaultValue=<rgba 0 0 0 1>}
  NamedRerouteDeclaration(None=Multiply(A=ScalarParameter() {group=00.Color, parameterName=Color_SHColorEmissiveScale}, B=Multiply(A=@n86, B=@n87.RGB))) {name=Emissive, nodeColor=<rgba 0.5 1 0 1>}
  NamedRerouteDeclaration(None=MaterialFunctionCall(BaseNormal=NamedRerouteUsage(), AdditionalNormal=NamedRerouteUsage()) {materialFunction=&BlendAngleCorrectedNormals}.Result) {name=Final_Normal, nodeColor=<rgba 0.1016 0 1 1>}
  NamedRerouteDeclaration(None=@n60) {name=ToTranslucent, nodeColor=<rgba 0.5469 1 0 1>}
  NamedRerouteDeclaration(None=@n9.UVs) {name=OriginUV, nodeColor=<rgba 1 0.966 0 1>}
  n88 = NamedRerouteUsage()
  n89 = Add(A=NamedRerouteUsage(), B=TextureCoordinate())
  n90 = CurveAtlasRowParameter(CurveTime=MaterialFunctionCall(Float2=Add(A=@n88, B=@n89)) {materialFunction=&BreakOutFloat2Components}.G) {
   atlas=&CLA_WaterFall
   curve=&CL_UVCurve
   group=00.Color
   parameterName=Curve_Color
  }
  NamedRerouteDeclaration(None=@n90.A) {name=CurveFoam_FoamMask, nodeColor=<rgba 0.6641 0 1 1>}
  NamedRerouteDeclaration() {name=CurveFoam_FoamColor, nodeColor=<rgba 0.6641 0 1 1>}
  NamedRerouteDeclaration(None=@n90) {name=CurveColor_Value, nodeColor=<rgba 1 0 0.1172 1>}
  NamedRerouteDeclaration(None=@n85) {name=Foam_GreyV, nodeColor=<rgba 1 0 0.8203 1>}
  n91 = VectorParameter() {group=05.Color, parameterName=TranslucentPartColor, defaultValue=<rgba 0 0 0 1>}
  n92 = LinearInterpolate(
   A=LinearInterpolate(A=NamedRerouteUsage(), B=@n87.RGB, Alpha=@n86)
   B=VectorParameter() {group=00.Color, parameterName=Color_FoamColor, defaultValue=<rgba 1 1 1 1>}.RGB
   Alpha=NamedRerouteUsage()
  )
  n93 = OneMinus(None=MaterialFunctionCall(Float2=TextureCoordinate()) {materialFunction=&BreakOutFloat2Components}.G)
  n94 = VectorParameter() {group=05.Color, parameterName=TranslucentPartOffsetAndWidth, defaultValue=<rgba -0.13 10 0 1>}
  n95 = Saturate(None=Multiply(A=Add(A=@n93, B=@n94.R), B=@n94.G))
  NamedRerouteDeclaration(None=Multiply(
    A=LinearInterpolate(A=@n91.RGB, B=@n92, Alpha=@n95)
    B=ScalarParameter() {defaultValue=1, group=05.Color, parameterName=TranslucentColorScale}
   )) {name=Final_BaseColor, nodeColor=<rgba 0 1 0.4844 1>}
  NamedRerouteDeclaration(None=@n28.R) {name=CurveA_UVRemap, nodeColor=<rgba 1 0.7031 0 1>}
  NamedRerouteDeclaration(None=LinearInterpolate(A=Constant3Vector() {constant=<rgba 0 0 1 0>}, B=NamedRerouteUsage(), Alpha=@n78)) {name=Foam_BlendedNormal, nodeColor=<rgba 0.7812 0 1 1>}
  n96 = VectorParameter() {
   group=10.Noise_FoamSpike
   parameterName=Noise_FoamSpike_Mask
   channelNames {a=U_Offset, b=V_Power, g=U_Contrast, r=U_Power}
   defaultValue=<rgba 0 0 0 1>
  }
  n97 = Constant()
  n98 = MaterialFunctionCall(Float2=Add(A=TextureCoordinate(), B=AppendVector(A=@n96.U_Offset, B=@n97))) {materialFunction=&BreakOutFloat2Components}
  n99 = Power(Base=@n98.G, Exp=@n96.V_Power)
  n100 = Max(A=@n98.R, B=OneMinus(None=@n98.R))
  n101 = Saturate(None=Multiply(A=Subtract(A=@n100) {constB=0.5}) {constB=2})
  n102 = MaterialFunctionCall(In=Power(Base=OneMinus(None=@n101), Exp=@n96.U_Power) {constExponent=5}, Contrast=@n96.U_Contrast) {materialFunction=&CheapContrast}
  n103 = NamedRerouteUsage()
  n104 = VectorParameter() {group=10.Noise_FoamSpike, parameterName=Noise_FoamSpike_ScaleAndSpeed, defaultValue=<rgba 1 1 0 0>}
  n105 = AppendVector(A=@n104.R, B=@n104.R)
  n106 = AppendVector(A=@n104.B, B=@n104.B)
  n107 = NamedRerouteUsage()
  n108 = TextureSample(UVs=Add(A=Multiply(A=@n103, B=@n105), B=Multiply(A=@n106, B=@n107))) {
   samplerType=SAMPLERTYPE_LinearColor
   texture=/Game/Athena/Artemis/Environments/BuildingActors/WaterFall/Materials/Texture/T_MergedNoiseTexture
  }
  n109 = Power(Base=@n108.A) {constExponent=1.5}
  n110 = VectorParameter() {group=10.Noise_FoamSpike, parameterName=Noise_FoamSpike_StrengthAndOffset, defaultValue=<rgba 1 1 0 0>}
  n111 = Multiply(
   A=Power(Base=AppendVector(A=@n109, B=@n109), Exp=ScalarParameter() {group=10.Noise_FoamSpike, parameterName=Noise_FoamSpike_Power})
   B=AppendVector(A=@n110.R, B=@n110.R)
  )
  n112 = AppendVector(A=@n110.B, B=@n110.B)
  NamedRerouteDeclaration(None=Multiply(A=Multiply(A=@n99, B=@n102.Result), B=Add(A=@n111, B=@n112))) {name=Noise_FoamSpike, nodeColor=<rgba 1 0 0.2578 1>}
  n113 = NamedRerouteUsage()
  n114 = VectorParameter() {group=10.Noise_Foam, parameterName=Noise_Foam_ScaleAndSpeed, defaultValue=<rgba 1 1 0 0>}
  n115 = AppendVector(A=@n114.R, B=@n114.R)
  n116 = AppendVector(A=@n114.B, B=@n114.B)
  n117 = NamedRerouteUsage()
  n118 = TextureSample(UVs=Add(A=Multiply(A=@n113, B=@n115), B=Multiply(A=@n116, B=@n117))) {
   samplerType=SAMPLERTYPE_LinearColor
   texture=&T_MergedNoiseTexture
  }
  n119 = Power(Base=@n118.A) {constExponent=1.5}
  n120 = AppendVector(A=@n119, B=@n119)
  n121 = VectorParameter() {group=10.Noise_Foam, parameterName=Noise_Foam_StrengthAndOffset, defaultValue=<rgba 1 1 0 0>}
  n122 = AppendVector(A=@n121.R, B=@n121.R)
  NamedRerouteDeclaration(None=Add(A=Multiply(A=@n120, B=@n122), B=AppendVector(A=@n121.B, B=@n121.B))) {name=Noise_Foam, nodeColor=<rgba 0.8281 1 0 1>}
  n123 = NamedRerouteUsage()
  n124 = VectorParameter() {group=10.Noise_ForTexture_A, parameterName=Noise_ForTexture_A_ScaleAndSpeed, defaultValue=<rgba 1 1 0 0>}
  n125 = AppendVector(A=@n124.R, B=@n124.R)
  n126 = AppendVector(A=@n124.B, B=@n124.B)
  n127 = NamedRerouteUsage()
  n128 = TextureSample(UVs=Add(A=Multiply(A=@n123, B=@n125), B=Multiply(A=@n126, B=@n127))) {
   samplerType=SAMPLERTYPE_LinearColor
   texture=&T_MergedNoiseTexture
  }
  n129 = MaterialFunctionCall(Float2=TextureCoordinate()) {materialFunction=&BreakOutFloat2Components}
  n130 = Saturate(None=Multiply(
    A=OneMinus(None=@n129.G)
    B=ScalarParameter() {defaultValue=2.25, group=10.Noise_ForTexture_A, parameterName=Noise_ForTexture_A_StrengthUVMaskPower}
   ) {constB=2.25})
  n131 = VectorParameter() {group=10.Noise_ForTexture_A, parameterName=Noise_ForTexture_A_StrengthAndOffset, defaultValue=<rgba 1 1 0 0>}
  n132 = AppendVector(A=@n131.R, B=@n131.R)
  n133 = Add(A=Multiply(A=AppendVector(A=@n128.B, B=@n128.B), B=Multiply(A=@n130, B=@n132)), B=AppendVector(A=@n131.B, B=@n131.B))
  n134 = Constant()
  n135 = NamedRerouteUsage()
  n136 = VectorParameter() {group=10.Noise_ForTexture_B, parameterName=Noise_ForTexture_B_ScaleAndSpeed, defaultValue=<rgba 1 1 0 0>}
  n137 = AppendVector(A=@n136.R, B=@n136.R)
  n138 = AppendVector(A=@n136.B, B=@n136.B)
  n139 = NamedRerouteUsage()
  n140 = TextureSample(UVs=Add(A=Multiply(A=@n135, B=@n137), B=Multiply(A=@n138, B=@n139))) {
   samplerType=SAMPLERTYPE_LinearColor
   texture=&T_MergedNoiseTexture
  }
  n141 = VectorParameter() {group=10.Noise_ForTexture_B, parameterName=Noise_ForTexture_B_StrengthAndOffset, defaultValue=<rgba 1 1 0 0>}
  n142 = Multiply(A=AppendVector(A=@n140.R, B=@n140.R), B=AppendVector(A=@n141.R, B=@n141.R))
  n143 = AppendVector(A=@n141.B, B=@n141.B)
  NamedRerouteDeclaration(None=Add(A=ShadingPathSwitch(Default=@n133, Deferred=@n133, Forward=@n133, Mobile=@n134), B=Add(A=@n142, B=@n143))) {name=Noise_ForTexture, nodeColor=<rgba 0.6172 1 0 1>}
  n144 = VectorParameter() {group=02.Normal, parameterName=NormalB_ScaleAndSpeed, defaultValue=<rgba 1 1 0 0>}
  n145 = Multiply(A=NamedRerouteUsage(), B=AppendVector(A=@n144.R, B=@n144.R))
  n146 = Multiply(A=AppendVector(A=@n144.B, B=@n144.B), B=NamedRerouteUsage())
  n147 = TextureSampleParameter2D(UVs=Add(A=NamedRerouteUsage(), B=Add(A=@n145, B=@n146))) {
   parameterName=NormalB_Texture
   samplerType=SAMPLERTYPE_Normal
   texture=/Game/Athena/Apollo/Water/Textures/02_Rounded/02_WaterFoam_Veroni_base_N
  }
  NamedRerouteDeclaration(None=MaterialFunctionCall(Normal=@n147.RGB, Flatness=ScalarParameter() {group=02.Normal, parameterName=NormalB_Flat}) {materialFunction=&FlattenNormal}.Result) {name=NormalB, nodeColor=<rgba 1 0.375 0 1>}
  n148 = VectorParameter() {group=01.FoamTexture, parameterName=FoamTexture_ScaleAndSpeed, defaultValue=<rgba 1 1 0 0>}
  n149 = Multiply(A=NamedRerouteUsage(), B=AppendVector(A=@n148.R, B=@n148.R))
  n150 = Multiply(A=AppendVector(A=@n148.B, B=@n148.B), B=NamedRerouteUsage())
  n151 = TextureSample(UVs=Add(A=NamedRerouteUsage(), B=Add(A=@n149, B=@n150))) {
   samplerType=SAMPLERTYPE_LinearColor
   texture=/Game/Athena/Artemis/Environments/BuildingActors/WaterFall/Materials/Texture/T_MergedFoamTexture
  }
  NamedRerouteDeclaration(None=@n151.B) {name=FoamTexture_GreyValue, nodeColor=<rgba 0 1 0.7656 1>}
  n152 = AppendVector(A=@n151.R, B=@n151.R)
  n153 = DeriveNormalZ(InXY=Multiply(A=Subtract(A=@n152) {constB=0.5}) {constB=2})
  NamedRerouteDeclaration(None=@n153) {name=FoamTexture_NormalOrigin, nodeColor=<rgba 0.3828 0 1 1>}
  NamedRerouteDeclaration(None=MaterialFunctionCall(Normal=@n153, Flatness=ScalarParameter() {group=01.FoamTexture, parameterName=FoamTexture_NormalFlat}) {materialFunction=&FlattenNormal}.Result) {name=FoamTexture_NormaFlat, nodeColor=<rgba 0.3828 0 1 1>}
  n154 = Power(Base=MaterialFunctionCall(Distance=ScalarParameter() {defaultValue=200, parameterName=CameraDistance}) {materialFunction=&Near_Camera_Fade}.Result) {constExponent=3}
  n155 = MaterialFunctionCall(Alpha Threshold=@n154) {materialFunction=&DitherTemporalAA}
  ShadowReplace(Default=Multiply(A=@n6, B=LinearInterpolate(A=@n155.Result, Alpha=@n154)), Shadow=@n6)
 ]
}
